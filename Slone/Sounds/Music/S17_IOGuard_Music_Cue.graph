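# Layered music cue for the S17 IO Guard encounter, routed through the Fort_Music_NPC sound class.
# A Mixer combines four looping stems (Idle, Base, Guard, Boss), with the Idle stem slightly attenuated.
# Each stem sits behind a continuous modulator whose volume follows a named parameter (Unaware, Base, Guard, Boss).
# The parameters default to zero, so external code sets them to blend layers while all stems play together.
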SoundCue /Slone/Sounds/Music/S17_IOGuard_Music_Cue
root=Mixer_1
nodes [
 {
  id=Mixer_1
  children [ModulatorContinuous_0, ModulatorContinuous_1, ModulatorContinuous_2, ModulatorContinuous_4]
  properties {inputVolume [0.7,1,1,1]}
 }
 {id=ModulatorContinuous_0, children [WavePlayer_6], properties {volumeModulationParams {default=0, parameterName=Unaware}}}
 {id=ModulatorContinuous_1, children [WavePlayer_3], properties {volumeModulationParams {default=0, parameterName=Base}}}
 {id=ModulatorContinuous_2, children [WavePlayer_5], properties {volumeModulationParams {default=0, parameterName=Guard}}}
 {id=ModulatorContinuous_4, children [WavePlayer_4], properties {volumeModulationParams {default=0, parameterName=Boss}}}
 {id=WavePlayer_3, properties {bLooping=true, soundWaveAssetPtr=&/S17_IOGuard_Music_Base}}
 {id=WavePlayer_4, properties {bLooping=true, soundWaveAssetPtr=&/S17_IOGuard_Music_Boss}}
 {id=WavePlayer_5, properties {bLooping=true, soundWaveAssetPtr=&/S17_IOGuard_Music_Guard}}
 {id=WavePlayer_6, properties {bLooping=true, soundWaveAssetPtr=&/S17_IOGuard_Music_Idle}}
]
properties {maxDistance=2097152, soundClassObject=&Fort_Music_NPC, volumeMultiplier=0.531}
